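# Sound cue for the happy vocalization of the MechOwl pet cosmetic.
# A mixer layers the Happy wave with a delayed branch that plays one of four equally weighted two-hoot variants one second later, each layer at reduced input volume.
# It uses the close-range pet attenuation and creature pet sound class, boosts overall volume, and is excluded from random-node branch culling.
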SoundCue /BRCosmetics/Sounds/Pets/MechEng/Pet_MechOwl_Happy_Cue
root=Mixer_0
nodes [
 {id=Delay_0, children [Random_0], properties {delayMax=1, delayMin=1}}
 {id=Mixer_0, children [WavePlayer_0,Delay_0], properties {inputVolume [0.2,0.3]}}
 {id=Random_0, children [WavePlayer_4, WavePlayer_11, WavePlayer_12, WavePlayer_13], properties {weights [1,1,1,1]}}
 {id=WavePlayer_0, properties {soundWaveAssetPtr=&/Pet_MechOwl_Happy}}
 {id=WavePlayer_11, properties {soundWaveAssetPtr=&/Pet_MechOwl_2Hoots_02}}
 {id=WavePlayer_12, properties {soundWaveAssetPtr=&/Pet_MechOwl_2Hoots_03}}
 {id=WavePlayer_13, properties {soundWaveAssetPtr=&/Pet_MechOwl_2Hoots_04}}
 {id=WavePlayer_4, properties {soundWaveAssetPtr=&/Pet_MechOwl_2Hoots_05}}
]
properties {
 attenuationSettings=&Fort_Pets_Close_Athena_Att
 bExcludeFromRandomNodeBranchCulling=true
 maxDistance=1800.0001
 soundClassObject=&Fort_Creatures_Pets
 volumeMultiplier=2.24
}
